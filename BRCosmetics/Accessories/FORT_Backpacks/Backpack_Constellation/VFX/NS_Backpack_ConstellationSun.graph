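# Niagara system for the sun effect on the Constellation backpack cosmetic: a glowing golden-orange sun with sparkles orbiting it.
# Flames_A and Inner layer rotated sprites drifting outward from small spheres, with fiery color-over-life curves and camera offset, while a single persistent particle adds an orange light.
# Dots and Stars_2 spawn gold sprites on a flattened hemisphere swirled by vortex velocity; stars add a second, animated glow renderer.
# Blue Stars_1 and larger Flames_A001 are disabled.
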NiagaraSystem /BRCosmetics/Accessories/FORT_Backpacks/Backpack_Constellation/VFX/NS_Backpack_ConstellationSun
dependencies {
 usedDataInterfaces [&NiagaraDataInterfaceCurve, &NiagaraDataInterfaceVector2DCurve, &NiagaraDataInterfaceColorCurve]
 usedDynamicInputs []
 usedModules [&SystemState]
}
emitters {
 Dots {
  emitterSpawnScript [&FN_PreserveSkydiveUserVariables() {module=PreserveUserVariables}]
  emitterUpdateScript [
   &EmitterState(
    Life Cycle Mode=::NewEnumerator1
    Loop Duration=&UniformRangedFloat(Minimum=0, Maximum=1)
    UseLoopDelay=false
   )
   &SpawnRate(SpawnRate=3, Use Spawn Probability=false)
  ]
  particleSpawnScript [
   /Niagara/Modules/Spawn/Initialization/InitializeParticle(
    Color=<rgba 0.9271 0.6902 0 1>
    Lifetime=1.25
    Module.Write Color=true
    Module.Write Lifetime=true
    Module.Write Mass=false
    Module.Write Material Random=true
    Module.Write Position=false
    Module.Write Scale=false
    Module.Write SpriteRotation=false
    Module.Write SpriteSize=true
    Sprite Size=&Vector2DFromFloat(Value=&UniformRangedFloat(Minimum=1, Maximum=2))
   )
   &SystemLocation(MaskSpawnGroup=false, Offset=<0 0 -2>)
   /Niagara/Modules/Spawn/Location/SphereLocation(
    Hemisphere Y=true
    MaskSpawnGroup=false
    Module.Override Randomness=true
    Module.Override Seed=true
    Non Uniform Scale=<-1 -0.5 1>
    Random Seed=2
    Randomness Mode:ENiagaraRandomnessMode=::NewEnumerator1
    Sphere Radius=12
    Surface Expansion Mode:ENiagaraExpansionMode=::NewEnumerator1
    Surface Only=true
   )
  ]
  particleUpdateScript [
   /Niagara/Modules/Update/Velocity/VortexVelocity(Module.Use Influence Falloff=false, Velocity Amount=10)
   &ParticleState()
   &SolveForcesAndVelocity(Clamp Velocity=false, Limit Acceleration=false)
   &ScaleSpriteSize(Scale Factor=&Vector2DFromFloat(Value=&FloatFromCurve(FloatCurve=DataInterface() {propertyValues {Curve {keys [<at 0.5 1>,<at 1 0>]}}})))
   &ScaleColor(
    Scale Alpha=&FloatFromCurve(FloatCurve=DataInterface() {propertyValues {Curve {keys [<at 0 0>,<at 0.1 1>,<at 1 0>]}}})
    ScaleA=true
    ScaleRGB=true
   )
  ]
  rendererClasses [&NiagaraSpriteRendererProperties]
  renderers [{rendererClass=&NiagaraSpriteRendererProperties}]
 }
 Flames_A {
  emitterUpdateScript [
   &EmitterState(Life Cycle Mode=::NewEnumerator1, Loop Duration=1.5, UseLoopDelay=false)
   &SpawnRate(SpawnRate=10, Use Spawn Probability=false)
   &SpawnBurst_Instantaneous(Spawn Count=2, Use Spawn Probability=false) {enabled=false}
  ]
  particleSpawnScript [
   &InitializeParticle(
    Lifetime=&UniformRangedFloat(Minimum=3, Maximum=3)
    Module.Write Color=true
    Module.Write Lifetime=true
    Module.Write Mass=false
    Module.Write Material Random=true
    Module.Write Position=true
    Module.Write Scale=false
    Module.Write SpriteRotation=true
    Module.Write SpriteSize=true
    Position=&SimulationPosition()
    Sprite Rotation=&UniformRangedFloat(Minimum=0, Maximum=360)
   )
   &SystemLocation(MaskSpawnGroup=false, Offset=<0 -8 -2>) {enabled=false}
   &DynamicMaterialParameters(
    Index 0 Param 1=&UniformRangedFloat(Minimum=0, Maximum=1)
    Index 0 Param 3=1
    Index 0 Param 4=1
    Module.Param0WriteEnabled=true
    Module.Param1WriteEnabled=false
    Module.Param2WriteEnabled=true
    Module.Param3WriteEnabled=true
   ) {enabled=false, module=DynamicMaterialParameters001}
   &SphereLocation(MaskSpawnGroup=false, Module.Override Randomness=false, Non Uniform Scale=<1 0.1 1>, Sphere Radius=2, Surface Only=false)
   &AddVelocity(Velocity=&Multiply_Vector(A=$Output.SphereLocation.SphereNormal, B=&VectorFromFloat(Value=5))) {enabled=false}
   /Niagara/Modules/Update/SubUV/SubUVAnimation(Number Of Frames=1, SubUV Animation Mode=::NewEnumerator1)
   &AddVelocityFromPoint(Module.Use Falloff Distance=false, Velocity Strength=0.5)
  ]
  particleUpdateScript [
   &ParticleState()
   &SolveForcesAndVelocity(Clamp Velocity=false, Limit Acceleration=false)
   &Color(
    Color=&ColorFromCurve(ColorCurve=DataInterface() {propertyValues {
       AlphaCurve {keys [<at 0 0 RCIM_Cubic>,<at 0.3 0.4 RCIM_Cubic>,<at 1 0 RCIM_Cubic>]}
       BlueCurve {keys [<at -0.0007 0>,<at 0.25 0.4626>,<at 0.649 0>,<at 1 0.0198 RCIM_Cubic>]}
       GreenCurve {keys [<at -0.0007 1.0144>,<at 0.25 0.763>,<at 0.649 0.0802>,<at 1 0.0198 RCIM_Cubic>]}
       RedCurve {keys [<at -0.0007 1.5>,<at 0.25 1>,<at 0.649 1>,<at 1 0.2 RCIM_Cubic>]}
      }})
    Scale Alpha=0.25
   )
   &ScaleSpriteSize(Scale Factor=&Vector2DFromFloat(Value=&FloatFromCurve(FloatCurve=DataInterface() {propertyValues {Curve {keys [
          <at 0 0.5055 RCIM_Cubic>
          {arriveTangent=-0.0055, interpMode=RCIM_Cubic, leaveTangent=-0.0055, time=0.2248, value=0.7496}
          <at 1 0.5 RCIM_Cubic>
         ]}}})))
   &CameraOffset(Camera Offset Amount=15)
  ]
  rendererClasses [&NiagaraSpriteRendererProperties]
  renderers [{rendererClass=&NiagaraSpriteRendererProperties}]
 }
 Flames_A001 {
  bEnabled=false
  emitterUpdateScript [
   &EmitterState(Life Cycle Mode=::NewEnumerator1, Loop Duration=1.5, UseLoopDelay=false)
   &SpawnRate(SpawnRate=10, Use Spawn Probability=false)
   &SpawnBurst_Instantaneous(Spawn Count=2, Use Spawn Probability=false) {enabled=false}
  ]
  particleSpawnScript [
   &InitializeParticle(
    Lifetime=&UniformRangedFloat(Minimum=3.5, Maximum=4)
    Module.Write Color=true
    Module.Write Lifetime=true
    Module.Write Mass=false
    Module.Write Material Random=true
    Module.Write Position=true
    Module.Write Scale=false
    Module.Write SpriteRotation=true
    Module.Write SpriteSize=true
    Position=&SimulationPosition()
    Sprite Rotation=&UniformRangedFloat(Minimum=0, Maximum=360)
   )
   &SystemLocation(MaskSpawnGroup=false, Offset=<0 -8 -2>) {enabled=false}
   &DynamicMaterialParameters(
    Index 0 Param 1=&UniformRangedFloat(Minimum=0, Maximum=1)
    Index 0 Param 3=1
    Index 0 Param 4=1
    Module.Param0WriteEnabled=true
    Module.Param1WriteEnabled=false
    Module.Param2WriteEnabled=true
    Module.Param3WriteEnabled=true
   ) {enabled=false, module=DynamicMaterialParameters001}
   &SphereLocation(MaskSpawnGroup=false, Module.Override Randomness=false, Sphere Radius=4, Surface Only=false)
   &AddVelocity(Velocity=&Multiply_Vector(A=$Output.SphereLocation.SphereNormal, B=&VectorFromFloat(Value=5))) {enabled=false}
   &SubUVAnimation(Number Of Frames=1, SubUV Animation Mode=::NewEnumerator1)
   &AddVelocityFromPoint(Module.Use Falloff Distance=false, Velocity Strength=2.5)
  ]
  particleUpdateScript [
   &ParticleState()
   &SolveForcesAndVelocity(Clamp Velocity=false, Limit Acceleration=false)
   &Color(
    Color=&ColorFromCurve(ColorCurve=DataInterface() {propertyValues {
       AlphaCurve {keys [<at 0 0 RCIM_Cubic>,<at 0.3 0.4 RCIM_Cubic>,<at 1 0 RCIM_Cubic>]}
       BlueCurve {keys [<at -0.0007 0>,<at 0.25 0.4626>,<at 0.649 0>,<at 1 0.0198 RCIM_Cubic>]}
       GreenCurve {keys [<at -0.0007 1.0144>,<at 0.25 0.763>,<at 0.649 0.0802>,<at 1 0.0198 RCIM_Cubic>]}
       RedCurve {keys [<at -0.0007 1.5>,<at 0.25 1>,<at 0.649 1>,<at 1 0.2 RCIM_Cubic>]}
      }})
    Scale Alpha=0.25
   )
   &ScaleSpriteSize(Scale Factor=&Vector2DFromFloat(Value=&FloatFromCurve(FloatCurve=DataInterface() {propertyValues {Curve {keys [<at 0 1 RCIM_Cubic>, {arriveTangent=1, interpMode=RCIM_Cubic, leaveTangent=1, time=0.2, value=1.5}, <at 1 2 RCIM_Cubic>]}}})))
   &CameraOffset(Camera Offset Amount=15)
  ]
  rendererClasses [&NiagaraSpriteRendererProperties]
  renderers [{rendererClass=&NiagaraSpriteRendererProperties}]
 }
 Inner {
  emitterUpdateScript [
   &EmitterState(Life Cycle Mode=::NewEnumerator1, Loop Duration=1.5, UseLoopDelay=false)
   &SpawnRate(SpawnRate=10, Use Spawn Probability=false)
   &SpawnBurst_Instantaneous(Spawn Count=2, Use Spawn Probability=false) {enabled=false}
  ]
  particleSpawnScript [
   &InitializeParticle(
    Lifetime=&UniformRangedFloat(Minimum=3.5, Maximum=4)
    Module.Write Color=true
    Module.Write Lifetime=true
    Module.Write Mass=false
    Module.Write Material Random=true
    Module.Write Position=true
    Module.Write Scale=false
    Module.Write SpriteRotation=true
    Module.Write SpriteSize=true
    Position=&SimulationPosition()
    Sprite Rotation=&UniformRangedFloat(Minimum=0, Maximum=360)
    Sprite Size=&UniformRangedVector2D(Minimum=<2 2>, Maximum=<4 4>)
   )
   &SystemLocation(MaskSpawnGroup=false, Offset=<0 -8 -2>) {enabled=false}
   &DynamicMaterialParameters(
    Index 0 Param 1=&UniformRangedFloat(Minimum=0, Maximum=1)
    Index 0 Param 3=1
    Index 0 Param 4=1
    Module.Param0WriteEnabled=true
    Module.Param1WriteEnabled=false
    Module.Param2WriteEnabled=true
    Module.Param3WriteEnabled=true
   ) {enabled=false, module=DynamicMaterialParameters001}
   &SphereLocation(MaskSpawnGroup=false, Module.Override Randomness=false, Sphere Radius=1.5, Surface Only=false)
   &AddVelocity(Velocity=&Multiply_Vector(A=$Output.SphereLocation.SphereNormal, B=&VectorFromFloat(Value=5))) {enabled=false}
   &SubUVAnimation(Number Of Frames=1, SubUV Animation Mode=::NewEnumerator1)
   &AddVelocityFromPoint(Module.Use Falloff Distance=false, Velocity Strength=1.25)
  ]
  particleUpdateScript [
   &ParticleState()
   &SolveForcesAndVelocity(Clamp Velocity=false, Limit Acceleration=false)
   &Color(Color=&ColorFromCurve(ColorCurve=DataInterface() {propertyValues {
       AlphaCurve {keys [<at 0 0 RCIM_Cubic>,<at 0.3 0.4 RCIM_Cubic>,<at 1 0 RCIM_Cubic>]}
       BlueCurve {keys [<at -0.0007 0>,<at 0.25 0.4626>,<at 0.649 0>,<at 1 0.0198 RCIM_Cubic>]}
       GreenCurve {keys [<at -0.0007 2.7051>,<at 0.25 0.763>,<at 0.649 0.0802>,<at 1 0.0198 RCIM_Cubic>]}
       RedCurve {keys [<at -0.0007 4>,<at 0.25 1>,<at 0.649 1>,<at 1 0.2 RCIM_Cubic>]}
      }}))
   &ScaleSpriteSize(Scale Factor=&Vector2DFromFloat(Value=&FloatFromCurve(FloatCurve=DataInterface() {propertyValues {Curve {keys [<at 0 1 RCIM_Cubic>, {arriveTangent=1, interpMode=RCIM_Cubic, leaveTangent=1, time=0.2, value=1.5}, <at 1 2 RCIM_Cubic>]}}})))
   &CameraOffset(Camera Offset Amount=15)
  ]
  rendererClasses [&NiagaraSpriteRendererProperties]
  renderers [{rendererClass=&NiagaraSpriteRendererProperties}]
 }
 Light {
  emitterSpawnScript [&FN_PreserveSkydiveUserVariables() {module=PreserveUserVariables}]
  emitterUpdateScript [
   &EmitterState(Life Cycle Mode=::NewEnumerator1, Loop Behavior=::NewEnumerator1, Loop Duration=1, UseLoopDelay=false)
   &SpawnBurst_Instantaneous(Spawn Count=1, Use Spawn Probability=true)
  ]
  particleSpawnScript [
   &InitializeParticle(
    Mesh Scale=<0.05 0.075 0.075>
    Module.Write Color=true
    Module.Write Lifetime=true
    Module.Write Mass=false
    Module.Write Material Random=true
    Module.Write Position=false
    Module.Write Scale=true
    Module.Write SpriteRotation=false
    Module.Write SpriteSize=false
   )
   &FN_Light_v2(Light Exponent=2, Light Radius=40, LightEnabled=true, Scale Color=1)
   /SetVariables(Particles.Light Color:LinearColor=<rgba 4 0.557 0 1>) {isSetParameters=true}
   &SystemLocation(MaskSpawnGroup=false, Offset=<0 -20 0>)
  ]
  particleUpdateScript [
   &ParticleState(Kill Particles When Lifetime Has Elapsed=false)
   &SolveForcesAndVelocity(Clamp Velocity=false, Limit Acceleration=false)
  ]
  rendererClasses [&NiagaraLightRendererProperties]
  renderers [{rendererClass=&NiagaraLightRendererProperties, rendererIndex=0}]
 }
 Stars_1 {
  bEnabled=false
  emitterSpawnScript [&FN_PreserveSkydiveUserVariables() {module=PreserveUserVariables}]
  emitterUpdateScript [
   &EmitterState()
   &SpawnRate(SpawnRate=2, Use Spawn Probability=false)
  ]
  particleSpawnScript [
   &InitializeParticle(
    Color=<rgba 0 0.3286 0.4948 1>
    Lifetime=1.25
    Module.Write Color=true
    Module.Write Lifetime=true
    Module.Write Mass=false
    Module.Write Material Random=true
    Module.Write Position=false
    Module.Write Scale=false
    Module.Write SpriteRotation=false
    Module.Write SpriteSize=true
    Sprite Size=<4 4>
   )
   &SystemLocation(MaskSpawnGroup=false, Offset=<0 0 -2>)
   &SphereLocation(
    Hemisphere Y=true
    MaskSpawnGroup=false
    Module.Override Randomness=false
    Non Uniform Scale=<-1 -0.5 1>
    Sphere Radius=12
    Surface Expansion Mode:ENiagaraExpansionMode=::NewEnumerator1
    Surface Only=true
   )
   /SetVariables(Particles.StarGlowColor:LinearColor=<rgba 0 0.6642 1 0.075>, Particles.StarGlowSpriteSize:Vector2f=<6 6>) {isSetParameters=true, module=SetVariables_#1}
  ]
  particleUpdateScript [
   &VortexVelocity(Module.Use Influence Falloff=false, Velocity Amount=5)
   &ParticleState()
   &SolveForcesAndVelocity(Clamp Velocity=false, Limit Acceleration=false)
   /SetVariables(
    Particles.StarGlowColor:LinearColor=&MultiplyLinearColor(
     Linear Color=$Particles.Initial.StarGlowColor
     Scale Factor=&MakeVector4(
      X=1
      Y=1
      Z=1
      W=&FloatFromCurve(FloatCurve=DataInterface() {propertyValues {Curve {keys [<at 0 1>,<at 1 0>]}}})
     )
    )
    Particles.StarGlowSpriteSize:Vector2f=&Multiply_Vector2(
     A=$Particles.Initial.StarGlowSpriteSize
     B=&Vector2DFromCurve(
      Vector2Curve=DataInterface() {propertyValues {
        XCurve {keys [
          {arriveTangent=50.4225, interpMode=RCIM_Cubic, leaveTangent=50.4232, tangentMode=RCTM_User, time=0, value=0.75}
          {arriveTangent=1, interpMode=RCIM_Cubic, leaveTangent=1, time=0.1, value=0.28}
          {arriveTangent=1.3585, interpMode=RCIM_Cubic, leaveTangent=1.3585, time=0.25, value=1}
          {arriveTangent=-1.3333, interpMode=RCIM_Cubic, leaveTangent=-1.3333, time=0.63, value=1}
          <at 1 0>
         ]}
        YCurve {keys [
          <at 0 2 RCIM_Cubic>
          {arriveTangent=-4, interpMode=RCIM_Cubic, leaveTangent=-4, time=0.1, value=1}
          <at 0.25 1>
          <at 0.2615 1 RCIM_Cubic>
          {arriveTangent=-1.354, interpMode=RCIM_Cubic, leaveTangent=-1.354, time=0.63, value=1}
          <at 1 0>
         ]}
       }}
      ScaleCurve=<1 1>
     )
    )
   ) {isSetParameters=true}
   &ScaleSpriteSize(Scale Factor=&Vector2DFromCurve(
     Vector2Curve=DataInterface() {propertyValues {
       XCurve {keys [
         {arriveTangent=50.4225, interpMode=RCIM_Cubic, leaveTangent=50.4232, tangentMode=RCTM_User, time=0, value=0.75}
         {arriveTangent=1, interpMode=RCIM_Cubic, leaveTangent=1, time=0.1, value=0.28}
         {arriveTangent=1.3585, interpMode=RCIM_Cubic, leaveTangent=1.3585, time=0.25, value=1}
         {arriveTangent=-1.3333, interpMode=RCIM_Cubic, leaveTangent=-1.3333, time=0.63, value=1}
         <at 1 0>
        ]}
       YCurve {keys [
         <at 0 2 RCIM_Cubic>
         {arriveTangent=-4, interpMode=RCIM_Cubic, leaveTangent=-4, time=0.1, value=1}
         <at 0.25 1>
         <at 0.2615 1 RCIM_Cubic>
         {arriveTangent=-1.354, interpMode=RCIM_Cubic, leaveTangent=-1.354, time=0.63, value=1}
         <at 1 0>
        ]}
      }}
     ScaleCurve=<1 1>
    ))
  ]
  rendererClasses [&NiagaraSpriteRendererProperties]
  renderers [
   {rendererClass=&NiagaraSpriteRendererProperties}
   {rendererClass=&NiagaraSpriteRendererProperties, rendererIndex=1}
  ]
 }
 Stars_2 {
  emitterSpawnScript [&FN_PreserveSkydiveUserVariables() {module=PreserveUserVariables}]
  emitterUpdateScript [
   &EmitterState(
    Life Cycle Mode=::NewEnumerator1
    Loop Duration=&UniformRangedFloat(Minimum=0, Maximum=1)
    UseLoopDelay=false
   )
   &SpawnRate(Use Spawn Probability=false)
  ]
  particleSpawnScript [
   &InitializeParticle(
    Color=<rgba 0.9271 0.6902 0 1>
    Lifetime=1.25
    Module.Write Color=true
    Module.Write Lifetime=true
    Module.Write Mass=false
    Module.Write Material Random=true
    Module.Write Position=false
    Module.Write Scale=false
    Module.Write SpriteRotation=false
    Module.Write SpriteSize=true
    Sprite Size=<2 2>
   )
   &SystemLocation(MaskSpawnGroup=false, Offset=<0 0 -2>)
   &SphereLocation(
    Hemisphere Y=true
    MaskSpawnGroup=false
    Module.Override Randomness=true
    Module.Override Seed=true
    Non Uniform Scale=<-1 -0.5 1>
    Random Seed=1
    Randomness Mode:ENiagaraRandomnessMode=::NewEnumerator1
    Sphere Radius=12
    Surface Expansion Mode:ENiagaraExpansionMode=::NewEnumerator1
    Surface Only=true
   )
   /SetVariables(Particles.StarGlowColor:LinearColor=<rgba 0.9271 0.6902 0 0.075>, Particles.StarGlowSpriteSize:Vector2f=<4 4>) {isSetParameters=true, module=SetVariables_#1}
  ]
  particleUpdateScript [
   &VortexVelocity(Module.Use Influence Falloff=false, Velocity Amount=5)
   &ParticleState()
   &SolveForcesAndVelocity(Clamp Velocity=false, Limit Acceleration=false)
   /SetVariables(
    Particles.StarGlowColor:LinearColor=&MultiplyLinearColor(
     Linear Color=$Particles.Initial.StarGlowColor
     Scale Factor=&MakeVector4(
      X=1
      Y=1
      Z=1
      W=&FloatFromCurve(FloatCurve=DataInterface() {propertyValues {Curve {keys [<at 0 1>,<at 1 0>]}}})
     )
    )
    Particles.StarGlowSpriteSize:Vector2f=&Multiply_Vector2(
     A=$Particles.Initial.StarGlowSpriteSize
     B=&Vector2DFromCurve(
      Vector2Curve=DataInterface() {propertyValues {
        XCurve {keys [
          {arriveTangent=50.4225, interpMode=RCIM_Cubic, leaveTangent=50.4232, tangentMode=RCTM_User, time=0, value=0.75}
          {arriveTangent=1, interpMode=RCIM_Cubic, leaveTangent=1, time=0.1, value=0.28}
          {arriveTangent=1.3585, interpMode=RCIM_Cubic, leaveTangent=1.3585, time=0.25, value=1}
          {arriveTangent=-1.3333, interpMode=RCIM_Cubic, leaveTangent=-1.3333, time=0.63, value=1}
          <at 1 0>
         ]}
        YCurve {keys [
          <at 0 2 RCIM_Cubic>
          {arriveTangent=-4, interpMode=RCIM_Cubic, leaveTangent=-4, time=0.1, value=1}
          <at 0.25 1>
          <at 0.2615 1 RCIM_Cubic>
          {arriveTangent=-1.354, interpMode=RCIM_Cubic, leaveTangent=-1.354, time=0.63, value=1}
          <at 1 0>
         ]}
       }}
      ScaleCurve=<1 1>
     )
    )
   ) {isSetParameters=true}
   &ScaleSpriteSize(Scale Factor=&Vector2DFromCurve(
     Vector2Curve=DataInterface() {propertyValues {
       XCurve {keys [
         {arriveTangent=50.4225, interpMode=RCIM_Cubic, leaveTangent=50.4232, tangentMode=RCTM_User, time=0, value=0.75}
         {arriveTangent=1, interpMode=RCIM_Cubic, leaveTangent=1, time=0.1, value=0.28}
         {arriveTangent=1.3585, interpMode=RCIM_Cubic, leaveTangent=1.3585, time=0.25, value=1}
         {arriveTangent=-1.3333, interpMode=RCIM_Cubic, leaveTangent=-1.3333, time=0.63, value=1}
         <at 1 0>
        ]}
       YCurve {keys [
         <at 0 2 RCIM_Cubic>
         {arriveTangent=-4, interpMode=RCIM_Cubic, leaveTangent=-4, time=0.1, value=1}
         <at 0.25 1>
         <at 0.2615 1 RCIM_Cubic>
         {arriveTangent=-1.354, interpMode=RCIM_Cubic, leaveTangent=-1.354, time=0.63, value=1}
         <at 1 0>
        ]}
      }}
     ScaleCurve=<1 1>
    ))
  ]
  rendererClasses [&NiagaraSpriteRendererProperties]
  renderers [
   {rendererClass=&NiagaraSpriteRendererProperties}
   {rendererClass=&NiagaraSpriteRendererProperties, rendererIndex=1}
  ]
 }
}
summary {
 systemName=NS_Backpack_ConstellationSun
 emitters [
  {emitterName=Light, rendererClasses [&NiagaraLightRendererProperties]}
  {bEnabled=false, emitterName=Stars_1, rendererClasses [&NiagaraSpriteRendererProperties]}
  {emitterName=Stars_2, rendererClasses [&NiagaraSpriteRendererProperties]}
  {emitterName=Dots, rendererClasses [&NiagaraSpriteRendererProperties]}
  {emitterName=Flames_A, rendererClasses [&NiagaraSpriteRendererProperties]}
  {emitterName=Inner, rendererClasses [&NiagaraSpriteRendererProperties]}
  {bEnabled=false, emitterName=Flames_A001, rendererClasses [&NiagaraSpriteRendererProperties]}
 ]
 userVariables [
  User.bIsFrontEndPreview {type=NiagaraBool}
  User.MinDistance {type=NiagaraFloat, defaultValue=0, description=""}
  User.PlayerSpeed {type=NiagaraFloat, defaultValue=0, description=""}
  User.RibbonIDOffset {type=NiagaraInt32}
  User.SkeletalMesh {
   type=NiagaraDataInterfaceSkeletalMesh
   defaultValue {
    struct=NiagaraExt_VariableValue_DataInterface
    value {dataInterface=&self:NiagaraDataInterfaceSkeletalMesh_0}
   }
  }
 ]
}
